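; Chain Whip Mantis melee ability that runs attack techniques with movement-aware montages, root motion warping and camera offsets.
; Activation records the movement mode (sprint, tactical sprint or default), which scales warp distance and camera offset.
; Input angle against actor forward picks stationary, forward, left or right montages and clamps warp direction.
; Server damage-window targeting queues hits, applying filtered effect containers a hotfixable count per frame and flushing leftovers on end.
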(event Ability|EventActivateAbility
  ;[0.01106 0.05729 0.01818] ActivateAbility: Figure out the movement mode of the character at activation.
  (Utilities|FlowControl|Sequence
    (:then_0
      (bind _outfortplayerpawn (CallFunction|GetAvatarasValidFortPlayerPawn)
        (:ValidFortPlayerPawn
          (if (Variables|/Script/FortniteGame.FortPlayerPawn|GetbIsTacticalSprinting _outfortplayerpawn)
            (Variables|Transient|SetMovementModeAtActivation "NewEnumerator2")
            (elif (PawnStats|IsSprinting _outfortplayerpawn)
              (Variables|Transient|SetMovementModeAtActivation "NewEnumerator1")
              (else
                (Variables|Transient|SetMovementModeAtActivation "NewEnumerator0")))))))
    (:then_1
      (|CallParentFunction))))

(event Ability|EventOnEndAbility
  (params (bWasCancelled :type Boolean))
  (bind _targetdataindex (Variables|Transient|GetTargetDataIndex))
  (bind _targetdatas (Variables|Transient|GetTargetDatas))
  ;[0.1146 0.06811 0.01653] EndAbility: Clean up and reset ability.
  (Utilities|FlowControl|Sequence
    (:then_0
      (Variables|Transient|SetMovementModeAtActivation "NewEnumerator0")
      (Utilities|FlowControl|Sequence
        (:then_0
          (bind _asynctask_damagetargeting (Variables|Transient|GetAsyncTask_DamageTargeting)
            (:then
              (GameplayTasks|EndTask _asynctask_damagetargeting))
            (:else)))
        (:then_1
          (Ability|HasAuthority
            (:True
              (CallFunction|ResetProcessTargetingData) ; Reset process targeting data so it can be entered next activation.
              ;;[0.01251 0.1094 0.1072] Damage any pending target datas and clear the array.
              (while (Utilities|Array|IsValidIndex _targetdatas _targetdataindex)
                (CallFunction|ApplyDamagetoTargetingData (Utilities|Array|Get(aref) _targetdatas _targetdataindex))
                (Math|Integer|IncrementInt _targetdataindex))
              (Utilities|Array|Clear (Variables|Transient|GetTargetDatas)))))))
    (:then_1
      (|CallParentFunction bWasCancelled))))

(event Mantis|EventOnTechniqueNotifyWindowStarted
  (params (NotifyWindow :type EFortMantisNotifyWindow%20Enum :default "Invalid") (Notify :type Mantis%20Window%20Object%20Reference))
  ;[0.1035 0.09646 0.2188] OnTechniqueNotifyWindowStarted: Perform targeting and apply damage.
  (|CallParentFunction NotifyWindow Notify)
  (Ability|HasAuthority
    (:True
      (Utilities|IsValid Notify
        (:Is%20Valid
          (if (Utilities|Enum|Equal(Enum) NotifyWindow "Damage")
            (bind (_asynctaskproxy _targetdata) (Ability|Tasks|PerformTargetingAbilityTask "Task_DamageTargeting" (Variables|Settings|GetDamageTargetingPreset) true)
              (:then
                (Variables|Transient|SetAsyncTask_DamageTargeting _asynctaskproxy))
              (:Targeted
                _)
              (:Cancelled
                _))
            ;;[0.1136 1 0.9602] Found a target
            (for _index (range (Ability|TargetData|GetDataCountfromTargetData _targetdata))
              (Utilities|Array|Add (Variables|Transient|GetTargetDatas) (Utilities|Struct|MakeChainWhipMantisTargetingData (Variables|/Script/MantisRuntime.FortAnimNotifyState_Mantis|GetMantisNotifyTag Notify) (Ability|TargetData|GetHitResultfromTargetData _targetdata _index))) ; Immediately add new targeting data to the end of the array.
            )
            (CallFunction|ProcessTargetingData) ; If we aren't already, start processing stored targeting data.
          ))))))

(event Custom|ProcessTargetingData
  (bind _targetdataindex (Variables|Transient|GetTargetDataIndex))
  (bind _remainingtargetdatastoprocessthisframe (Variables|Transient|GetRemainingTargetDatasToProcessThisFrame))
  (bind _targetdatas (Variables|Transient|GetTargetDatas))
  ;[0.03164 0.8097 1 0.439] ProcessTargetingData
  ; Processes X pieces of targeting data per frame until all are done.
  (Utilities|FlowControl|DoOnce false
    (:Completed
      (Variables|Transient|SetTargetDataIndex 0)
      (CallFunction|ResetRemainingTargetDatastoProcessThisFrame)
      ;;[0.03505 1 0] Start of Loop
      (if (Utilities|Array|IsValidIndex _targetdatas _targetdataindex)
        (Utilities|FlowControl|Sequence
          (:then_0
            _)
          (:then_1
            ;;[0.1099 0.184 1] Increment the loop and either:
            ;; A) continue the loop immediately
            ;; or B) wait one frame before continuing the loop
            (Math|Integer|IncrementInt _targetdataindex
              (:%20%20
                (Math|Integer|DecrementInt _remainingtargetdatastoprocessthisframe
                  (:%20%20
                    (if (> _remainingtargetdatastoprocessthisframe 0)
                      _)
                    ;;;[1 0.9647 0.2921] We have no more remaining processes this frame, wait 1 frame before continuing the loop.
                    (Utilities|FlowControl|DelayUntilNextTick) ; Note: Timers are cleared on the end of abilities.
                    (CallFunction|ResetRemainingTargetDatastoProcessThisFrame)))))))
        ;;[0 0.8309 1 0.7] Loop Body
        (CallFunction|ApplyDamagetoTargetingData (Utilities|Array|Get(aref) _targetdatas _targetdataindex)))
      ;;[1 0.599 0.1043] End Loop and Clear Data, and reset our ability to start a new loop.
      (Utilities|Array|Clear _targetdatas)
      (CallFunction|ResetProcessTargetingData))) ; Do not restart until we have processed all current requests
)

(event Custom|ResetProcessTargetingData
  (bind _targetdataindex (Variables|Transient|GetTargetDataIndex))
  (bind _remainingtargetdatastoprocessthisframe (Variables|Transient|GetRemainingTargetDatasToProcessThisFrame))
  (bind _targetdatas (Variables|Transient|GetTargetDatas))
  ;[0.03164 0.8097 1 0.439] ProcessTargetingData
  ; Processes X pieces of targeting data per frame until all are done.
  (Utilities|FlowControl|DoOnce false
    (:Completed
      (Variables|Transient|SetTargetDataIndex 0)
      (CallFunction|ResetRemainingTargetDatastoProcessThisFrame)
      ;;[0.03505 1 0] Start of Loop
      (if (Utilities|Array|IsValidIndex _targetdatas _targetdataindex)
        (Utilities|FlowControl|Sequence
          (:then_0
            _)
          (:then_1
            ;;[0.1099 0.184 1] Increment the loop and either:
            ;; A) continue the loop immediately
            ;; or B) wait one frame before continuing the loop
            (Math|Integer|IncrementInt _targetdataindex
              (:%20%20
                (Math|Integer|DecrementInt _remainingtargetdatastoprocessthisframe
                  (:%20%20
                    (if (> _remainingtargetdatastoprocessthisframe 0)
                      _)
                    ;;;[1 0.9647 0.2921] We have no more remaining processes this frame, wait 1 frame before continuing the loop.
                    (Utilities|FlowControl|DelayUntilNextTick) ; Note: Timers are cleared on the end of abilities.
                    (CallFunction|ResetRemainingTargetDatastoProcessThisFrame)))))))
        ;;[0 0.8309 1 0.7] Loop Body
        (CallFunction|ApplyDamagetoTargetingData (Utilities|Array|Get(aref) _targetdatas _targetdataindex)))
      ;;[1 0.599 0.1043] End Loop and Clear Data, and reset our ability to start a new loop.
      (Utilities|Array|Clear _targetdatas)
      (CallFunction|ResetProcessTargetingData))) ; Do not restart until we have processed all current requests
)

(fn ApplyDamageToTargetingData
  (params (ChainWhipMantisTargetingData :type Chain%20Whip%20Mantis%20Targeting%20Data%20Structure%20%28by%20ref%29))
  (bind (_mantisnotifytag_2_32e0f3a9429577244f7c41a2b2bf8864 _hitresult_5_4e02ef6443ca41f1c18d8c919dc212d1) (Utilities|Struct|BreakChainWhipMantisTargetingData ChainWhipMantisTargetingData))
  ;[0.0208 0.125 0.1225] ApplyDamageToTargetingData: Apply gameplay effect container to the target data (holds damage GEs)
  (Ability|HasAuthority
    (:True
      (bind (_bvalidfilter _filter) (CallFunction|GetAttackTargetFilter))
      (if _bvalidfilter
        (bind _returnvalue (Ability|FilterSingleTargetHitResult _filter (CallFunction|GetAvatarasFortPlayerPawn) _hitresult_5_4e02ef6443ca41f1c18d8c919dc212d1))
        (if _returnvalue
          (Ability|ApplyGameplayEffectContainer (Ability|TargetData|AbilityTargetDatafromHitResult _hitresult_5_4e02ef6443ca41f1c18d8c919dc212d1) _mantisnotifytag_2_32e0f3a9429577244f7c41a2b2bf8864))))))

(fn OnMantisAblilityTechniqueStarted
  (params (OwningTags :type Gameplay%20Tag%20Container%20Structure) (SequenceName :type Name :default "None") (StartsSequence :type Boolean) (EndsSequence :type Boolean))
  (bind _cameramodifier (LocalVariables|GetCameraModifier))
  (bind _movementmodeatactivation (Variables|Transient|GetMovementModeAtActivation))
  ;[0.03467 0.2083 0.2042] MantisAbilityTechniqueStarted: Based on the started technique, update the camera offset.
  (|CallParentFunction OwningTags SequenceName StartsSequence EndsSequence)
  (if (Variables|Debug|GetbDebugTechnique)
    (Ability|HasAuthority
      (:True
        (Development|PrintString (Utilities|String|Append "ChainWhip: TechniqueName: " (Utilities|String|ToString(Name) SequenceName)) true false "(R=1.000000,G=0.248843,B=0.953802,A=1.000000)" 5.0))
      (:False
        _))
    (bind _asb_chain_whip (Utilities|Casting|CastToB_ChainWhip (Ability|GetCurrentSourceWeapon)))
    (bind _output_get (LocalVariables|SetCameraModifier (Class|BChainWhip|GetCameraModifier _asb_chain_whip)))
    (Utilities|IsValid _output_get
      (:Is%20Valid
        (bind _breturnvalue (CallFunction|ShouldUseMovingTechnique))
        (if _breturnvalue
          (switch name SequenceName
            (:FullBodyAttack1
              (Class|PlayerCameraModifierChainWhip|AddCameraOffset _cameramodifier (+ (Utilities|Select :Index _movementmodeatactivation :NewEnumerator0 0.0 :NewEnumerator1 0.0 :NewEnumerator2 1e+02) 1e+02)))
            (:FullBodyAttack2
              (Class|PlayerCameraModifierChainWhip|AddCameraOffset _cameramodifier (+ (Utilities|Select :Index _movementmodeatactivation :NewEnumerator0 0.0 :NewEnumerator1 0.0 :NewEnumerator2 75.0) 75.0)))
            (:FullBodyAttack3
              (Class|PlayerCameraModifierChainWhip|AddCameraOffset _cameramodifier (+ (Utilities|Select :Index _movementmodeatactivation :NewEnumerator0 0.0 :NewEnumerator1 0.0 :NewEnumerator2 1e+02) 175.0)))))))
    (else
      _)))

(fn ClampVectorWithAngles
  (params (TargetVector :type Vector) (ReferenceVector :type Vector) (ClampAngle :type Float%20%28double-precision%29))
  ;[0.03111 0.1263 0.1302] Clamps the target vector within clamp angle to the reference vector.
  (return (Math|Vector|RotateVectorAroundAxis (Variables|GetReferenceVector) (Math|Float|ClampAngle (Math|Trig|Acos(Degrees) (Math|Vector|DotProduct (Math|Vector|Normalize TargetVector) (Math|Vector|Normalize ReferenceVector))) (* ClampAngle -1.0) ClampAngle) (Math|Vector|Normalize (Math|Vector|CrossProduct (Variables|GetReferenceVector) (Variables|GetTargetVector))))))

(fn ConfigureRootMotionWarping
  (params (TechniqueName :type Name :default "None"))
  (bind _techniquenametoforwardrootmotionwarpmap (Variables|Settings|GetTechniqueNameToForwardRootMotionWarpMap))
  (bind _movementmodeatactivation (Variables|Transient|GetMovementModeAtActivation))
  (bind _warpingdirectionclampangle (Variables|Settings|GetWarpingDirectionClampAngle))
  (bind _techniquenametosidewaysrootmotionwarpmap (Variables|Settings|GetTechniqueNameToSidewaysRootMotionWarpMap))
  ;[0.03154 0.121 0.1667] ConfigureRootMotionWarping: Set up root motion warping based on the player input, if desired.
  (Utilities|FlowControl|Sequence
    (:then_0
      ;;[0.1259 0.08157 0.1979] Only for moving techniques.
      (bind (_breturnvalue _busesidewaysmovement) (CallFunction|ShouldUseMovingTechnique))
      (if _breturnvalue
        (if (.ReturnValue (Utilities|Map|Find (select _busesidewaysmovement _techniquenametosidewaysrootmotionwarpmap _techniquenametoforwardrootmotionwarpmap) TechniqueName))
          (bind _outfortplayerpawn (CallFunction|GetAvatarasValidFortPlayerPawn)
            (:ValidFortPlayerPawn
              (bind _charactermovement (Variables|/Script/Engine.Character|GetCharacterMovement _outfortplayerpawn)
                (:then
                  (if (Utilities|Operators|NotEqual(!=) :A (Pawn|Components|PawnMovement|GetLastInputVector _charactermovement) :ErrorTolerance 0.01)
                    (if (and (Variables|Debug|GetbDebugWarpMovement) (Editor|InEditororPIE))
                      (Development|PrintString (Utilities|String|Append "ChainWhip: MovementMode: " (Utilities|String|EnumtoString (Variables|Transient|GetMovementModeAtActivation)) " ; WarpDistance: " (Utilities|String|ToString(Float) (Utilities|Select :Index _movementmodeatactivation :NewEnumerator0 (Utilities|Struct|BreakChainWhipMantisRootMotionWarpingInfoStruct (Utilities|Map|Find (select _busesidewaysmovement _techniquenametosidewaysrootmotionwarpmap _techniquenametoforwardrootmotionwarpmap) TechniqueName)) :NewEnumerator1 (.Running_4_3024755D44CFB8C9CC1C4688B2E6C5ED (Utilities|Struct|BreakChainWhipMantisRootMotionWarpingInfoStruct (Utilities|Map|Find (select _busesidewaysmovement _techniquenametosidewaysrootmotionwarpmap _techniquenametoforwardrootmotionwarpmap) TechniqueName))) :NewEnumerator2 (.Sprinting_6_767CFBC5417462C7307328842959928B (Utilities|Struct|BreakChainWhipMantisRootMotionWarpingInfoStruct (Utilities|Map|Find (select _busesidewaysmovement _techniquenametosidewaysrootmotionwarpmap _techniquenametoforwardrootmotionwarpmap) TechniqueName)))))) true false "(R=0.000000,G=0.660000,B=1.000000,A=1.000000)" 5.0 "")
                      (Rendering|Debug|DrawDebugArrow (Transformation|GetActorLocation (CallFunction|GetAvatarasFortPlayerPawn)) (+ (Math|Vector|MakeVector (* (.x (Math|Vector|Normalize2D(Vector) (CallFunction|ClampVectorwithAngles (Pawn|Components|PawnMovement|GetLastInputVector _charactermovement) (Transformation|GetActorForwardVector _outfortplayerpawn) _warpingdirectionclampangle))) (Utilities|Select :Index _movementmodeatactivation :NewEnumerator0 (Utilities|Struct|BreakChainWhipMantisRootMotionWarpingInfoStruct (Utilities|Map|Find (select _busesidewaysmovement _techniquenametosidewaysrootmotionwarpmap _techniquenametoforwardrootmotionwarpmap) TechniqueName)) :NewEnumerator1 (.Running_4_3024755D44CFB8C9CC1C4688B2E6C5ED (Utilities|Struct|BreakChainWhipMantisRootMotionWarpingInfoStruct (Utilities|Map|Find (select _busesidewaysmovement _techniquenametosidewaysrootmotionwarpmap _techniquenametoforwardrootmotionwarpmap) TechniqueName))) :NewEnumerator2 (.Sprinting_6_767CFBC5417462C7307328842959928B (Utilities|Struct|BreakChainWhipMantisRootMotionWarpingInfoStruct (Utilities|Map|Find (select _busesidewaysmovement _techniquenametosidewaysrootmotionwarpmap _techniquenametoforwardrootmotionwarpmap) TechniqueName))))) (* (.y (Math|Vector|Normalize2D(Vector) (CallFunction|ClampVectorwithAngles (Pawn|Components|PawnMovement|GetLastInputVector _charactermovement) (Transformation|GetActorForwardVector _outfortplayerpawn) _warpingdirectionclampangle))) (Utilities|Select :Index _movementmodeatactivation :NewEnumerator0 (Utilities|Struct|BreakChainWhipMantisRootMotionWarpingInfoStruct (Utilities|Map|Find (select _busesidewaysmovement _techniquenametosidewaysrootmotionwarpmap _techniquenametoforwardrootmotionwarpmap) TechniqueName)) :NewEnumerator1 (.Running_4_3024755D44CFB8C9CC1C4688B2E6C5ED (Utilities|Struct|BreakChainWhipMantisRootMotionWarpingInfoStruct (Utilities|Map|Find (select _busesidewaysmovement _techniquenametosidewaysrootmotionwarpmap _techniquenametoforwardrootmotionwarpmap) TechniqueName))) :NewEnumerator2 (.Sprinting_6_767CFBC5417462C7307328842959928B (Utilities|Struct|BreakChainWhipMantisRootMotionWarpingInfoStruct (Utilities|Map|Find (select _busesidewaysmovement _techniquenametosidewaysrootmotionwarpmap _techniquenametoforwardrootmotionwarpmap) TechniqueName)))))) (Transformation|GetActorLocation _outfortplayerpawn)) 2e+02 "(R=1.000000,G=0.000000,B=0.250307,A=1.000000)" 4.0 5.0)
                      (else
                        _))
                    (return true false 0 false true (+ (Math|Vector|MakeVector (* (.x (Math|Vector|Normalize2D(Vector) (CallFunction|ClampVectorwithAngles (Pawn|Components|PawnMovement|GetLastInputVector _charactermovement) (Transformation|GetActorForwardVector _outfortplayerpawn) _warpingdirectionclampangle))) (Utilities|Select :Index _movementmodeatactivation :NewEnumerator0 (Utilities|Struct|BreakChainWhipMantisRootMotionWarpingInfoStruct (Utilities|Map|Find (select _busesidewaysmovement _techniquenametosidewaysrootmotionwarpmap _techniquenametoforwardrootmotionwarpmap) TechniqueName)) :NewEnumerator1 (.Running_4_3024755D44CFB8C9CC1C4688B2E6C5ED (Utilities|Struct|BreakChainWhipMantisRootMotionWarpingInfoStruct (Utilities|Map|Find (select _busesidewaysmovement _techniquenametosidewaysrootmotionwarpmap _techniquenametoforwardrootmotionwarpmap) TechniqueName))) :NewEnumerator2 (.Sprinting_6_767CFBC5417462C7307328842959928B (Utilities|Struct|BreakChainWhipMantisRootMotionWarpingInfoStruct (Utilities|Map|Find (select _busesidewaysmovement _techniquenametosidewaysrootmotionwarpmap _techniquenametoforwardrootmotionwarpmap) TechniqueName))))) (* (.y (Math|Vector|Normalize2D(Vector) (CallFunction|ClampVectorwithAngles (Pawn|Components|PawnMovement|GetLastInputVector _charactermovement) (Transformation|GetActorForwardVector _outfortplayerpawn) _warpingdirectionclampangle))) (Utilities|Select :Index _movementmodeatactivation :NewEnumerator0 (Utilities|Struct|BreakChainWhipMantisRootMotionWarpingInfoStruct (Utilities|Map|Find (select _busesidewaysmovement _techniquenametosidewaysrootmotionwarpmap _techniquenametoforwardrootmotionwarpmap) TechniqueName)) :NewEnumerator1 (.Running_4_3024755D44CFB8C9CC1C4688B2E6C5ED (Utilities|Struct|BreakChainWhipMantisRootMotionWarpingInfoStruct (Utilities|Map|Find (select _busesidewaysmovement _techniquenametosidewaysrootmotionwarpmap _techniquenametoforwardrootmotionwarpmap) TechniqueName))) :NewEnumerator2 (.Sprinting_6_767CFBC5417462C7307328842959928B (Utilities|Struct|BreakChainWhipMantisRootMotionWarpingInfoStruct (Utilities|Map|Find (select _busesidewaysmovement _techniquenametosidewaysrootmotionwarpmap _techniquenametoforwardrootmotionwarpmap) TechniqueName)))))) (Transformation|GetActorLocation _outfortplayerpawn)) false 0)))
                (:else)))))))
    (:then_1
      (return false false 0 false false "0, 0, 0" false 0))))

(fn GetAttackTargetFilter
  (bind _returnvalue (CallFunction|GetTargetingComponent))
  ;[0.07498 0.1585 0.1823] Returns the attack target filter struct from the targeting component.
  (Utilities|FlowControl|Sequence
    (:then_0
      (Utilities|IsValid _returnvalue
        (:Is%20Valid
          (bind _filter (Attack|GetAttackTargetFilter _returnvalue))
          (return true _filter))))
    (:then_1
      (return false 0))))

(fn GetDynamicMontageNameForTechnique
  (params (TechniqueName :type Name :default "None"))
  ;[0.09549 0.03364 0.1042] GetDynamicMontageNameForTechnique: Figure out if the stationary or moving montage should be used based on if the player is moving forward.
  (bind (_breturnvalue _busesidewaysmovement _buseleftsidewaysmovement) (CallFunction|ShouldUseMovingTechnique))
  (if _breturnvalue
    (if _busesidewaysmovement
      (if (Variables|Debug|GetbDebugWarpMovement)
        (Development|PrintString (Utilities|String|Append "ChainWhip: MontageName: " (Utilities|String|ToString(Name) (select _buseleftsidewaysmovement "MovingLeft" "MovingRight"))) true false "(R=0.000000,G=0.660000,B=1.000000,A=1.000000)" 5.0)
        (else
          _))
      (return (select _buseleftsidewaysmovement "MovingLeft" "MovingRight"))
      (else
        (if (Variables|Debug|GetbDebugWarpMovement)
          (Development|PrintString "ChainWhip: MontageName: MovingForward" true false "(R=0.000000,G=0.660000,B=1.000000,A=1.000000)" 5.0)
          (else
            _))
        (return "MovingForward")))
    (elif (Variables|Debug|GetbDebugWarpMovement)
      (Development|PrintString "ChainWhip: MontageName: Stationary" true false "(R=0.000000,G=0.660000,B=1.000000,A=1.000000)" 5.0)
      (return "Stationary")
      (else
        _))))

(fn GetTargetingComponent
  (bind _returnvalue (Ability|GetCurrentSourceWeapon))
  ;[0.03222 0.1392 0.1406] GetTargetingComponent: Returns the targeting component of the source weapon.
  (bind _targetingcomponent (Variables|Default|GetTargetingComponent)
    (:then
      (return _targetingcomponent))
    (:else
      (Utilities|IsValid _returnvalue
        (:Is%20Valid
          (bind _output_get (Variables|Default|SetTargetingComponent (Actor|GetComponentbyClass _returnvalue "/SunRoseChainWhipGameplay/Items/ChainWhip/B_ChainWhipTargetingComponent.B_ChainWhipTargetingComponent_C")))
          (return _output_get))))))

(fn ResetRemainingTargetDatasToProcessThisFrame
  (Variables|Transient|SetRemainingTargetDatasToProcessThisFrame (Math|Integer|Max(Integer) (ScalableFloat|GetValueasInteger (Variables|Hotfix|GetHotfix_NumTargetsToProcessEachFrame)) 1)))

(fn ShouldUseMovingTechnique
  (bind _returnvalue (CallFunction|GetAvatarasFortPlayerPawn))
  ;[0.009681 0.121 0.1458] ShouldUseMovingTechnique: Returns if the moving technique should be used based on player input.
  (Utilities|FlowControl|Sequence
    (:then_0
      (Utilities|IsValid _returnvalue
        (:Is%20Valid
          (bind _charactermovement (Variables|/Script/Engine.Character|GetCharacterMovement _returnvalue)
            (:then
              (if (Utilities|Operators|NotEqual(!=) :A (Pawn|Components|PawnMovement|GetLastInputVector _charactermovement) :ErrorTolerance 0.01)
                (return (Math|Float|InRange(Float) (Math|Trig|Acos(Degrees) (Math|Vector|DotProduct (Math|Vector|Normalize (Pawn|Components|PawnMovement|GetLastInputVector _charactermovement)) (Transformation|GetActorForwardVector _returnvalue))) 0.0 (Variables|Settings|GetMovingTechniqueMaxAngle)) (> (Math|Trig|Acos(Degrees) (Math|Vector|DotProduct (Math|Vector|Normalize (Pawn|Components|PawnMovement|GetLastInputVector _charactermovement)) (Transformation|GetActorForwardVector _returnvalue))) (Variables|Settings|GetDiagonalMovementInputAngleThreshold)) (> (.z (Math|Vector|CrossProduct (Math|Vector|Normalize (Pawn|Components|PawnMovement|GetLastInputVector _charactermovement)) (Transformation|GetActorForwardVector _returnvalue))) 0))))
            (:else)))))
    (:then_1
      (return false false false))))
