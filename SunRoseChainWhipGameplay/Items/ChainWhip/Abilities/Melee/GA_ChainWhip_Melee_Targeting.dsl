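; Gameplay ability that runs melee targeting for the Chain Whip to update its crosshair while active.
; A self-rescheduling timer runs a targeting task, with update frequency and async mode cached from hotfix values on activation.
; Results pass through the weapon targeting component's attack filter, setting the reticle target type to world or pawn and printing it for debugging.
; Ending the ability stops the timer and task and resets the reticle.
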(event Ability|EventActivateAbility
  ;[0.0428 0.2083 0.05471] ActivateAbility: Start running targeting traces to figure out what the melee attack is aiming.
  (|CallParentFunction)
  (CallFunction|InitHotfixes)
  (CallFunction|SetTargetingEnabled true))

(event Ability|EventOnEndAbility
  (params (bWasCancelled :type Boolean))
  ;[0.2031 0.1243 0.02589] EndAbility: Clean up.
  (Utilities|FlowControl|Sequence
    (:then_0
      (CallFunction|SetTargetingEnabled false)
      (bind _targetingcomponent (Variables|Transient|GetTargetingComponent)
        (:then
          (Reticle|Targeting|SetTargetType _targetingcomponent "NewEnumerator0"))
        (:else)))
    (:then_1
      (|CallParentFunction bWasCancelled))))

(event Custom|SetTargetingEnabled
  (params (bEnabled :type Boolean))
  ;[0.125 0.01902 0.09411] MeleeTargeting: Perform targeting and update the targeting component with the latest information.
  ; The targeting component in turn updates the crosshair.
  ;;[0 0.5208 0.1542] SetTargetingEnabled: Client only. If enabled, start performing melee targeting trace.
  (if bEnabled
    (CallFunction|PerformTargetingTrace)
    (else
      (Utilities|Time|ClearandInvalidateTimerbyHandle (Variables|Transient|GetTargetingTraceTimerHandle))
      (bind _asynctask_targeting (Variables|Transient|GetAsyncTask_Targeting)
        (:then
          (GameplayTasks|EndTask _asynctask_targeting))
        (:else)))))

(event Custom|PerformTargetingTrace
  (bind _outputdelegate (EventDispatchers|CreateEvent :!event Custom|PerformTargetingTrace))
  (bind _targetingupdatefreq (Variables|Transient|GetTargetingUpdateFreq))
  ;[0.125 0.01902 0.09411] MeleeTargeting: Perform targeting and update the targeting component with the latest information.
  ; The targeting component in turn updates the crosshair.
  ;;[0.08033 0.1667 0.1449] PerformTargetingTrace: Self calling function that executes the targeting task for the hookshot ability.
  (bind (_asynctaskproxy _targetdata) (Ability|Tasks|PerformTargetingAbilityTask "Task_MeleeTargeting" (Variables|Default|GetTargetingPresetMelee) false false (Variables|Transient|GetbUseAsyncTargeting))
    (:then
      (Variables|Transient|SetAsyncTask_Targeting _asynctaskproxy))
    (:Targeted
      _)
    (:Cancelled
      _))
  (Default|UpdateTargetingComponent _targetdata)
  (if (> _targetingupdatefreq 0)
    (bind _returnvalue (Utilities|Time|SetTimerbyEvent _outputdelegate _targetingupdatefreq false true))
    (Variables|Transient|SetTargetingTraceTimerHandle _returnvalue)
    (else
      (bind _returnvalue_2 (Utilities|Time|SetTimerforNextTickbyEvent _outputdelegate))
      (Variables|Transient|SetTargetingTraceTimerHandle _returnvalue_2))))

(fn InitHotfixes
  ;[0.05643 0.1788 0.2396] InitHotfixes: Cache the hotfix values so they don't have to constantly looked up .
  (Variables|Transient|SetbUseAsyncTargeting (ScalableFloat|GetValueasBool (Variables|Hotfix|GetHotfix_UseAsyncTargeting)))
  (Variables|Transient|SetTargetingUpdateFreq (ScalableFloat|GetValueatLevel (Variables|Hotfix|GetHotfix_TargetingUpdateFreq))))

(fn UpdateTargetingComponent
  (params (TargetData :type Gameplay%20Ability%20Target%20Data%20Handle%20Structure))
  (bind _returnvalue (Ability|GetCurrentSourceWeapon))
  (bind _targettype (LocalVariables|GetTargetType))
  ;[0.03183 0.1354 0.1246] UpdateTargetingComponent: Run the trace to find the potential target that the melee ability is going to land on and update the reticle through the target component.
  (Variables|Transient|GetTargetingComponent
    (:then
      _)
    (:else
      (Utilities|IsValid _returnvalue
        (:Is%20Valid
          (bind _output_get (Variables|Transient|SetTargetingComponent (Actor|GetComponentbyClass _returnvalue "/SunRoseChainWhipGameplay/Items/ChainWhip/B_ChainWhipTargetingComponent.B_ChainWhipTargetingComponent_C")))
          (Utilities|IsValid _output_get
            (:Is%20Valid
              _))))))
  (bind _targetingcomponent (Variables|Transient|GetTargetingComponent)
    (:then
      (Utilities|FlowControl|Sequence
        (:then_0
          (bind _filter (Attack|GetAttackTargetFilter _targetingcomponent))
          (bind (_batleastoneactorleft _returnvalue_2) (Ability|FilterTargetActorList _filter (CallFunction|GetAvatarasFortPlayerPawn) (Ability|TargetData|GetAllActorsfromTargetData TargetData)))
          ;;[0.6562 1 0.3125] If hit something, consider target type as world.
          (bind _output_get_2 (LocalVariables|SetTargetActors _returnvalue_2))
          (if (Utilities|Array|IsNotEmpty _output_get_2)
            (LocalVariables|SetTargetType "NewEnumerator1")
            ;;[0.1801 0.04091 0.1979] If hit a pawn, consider target type as Pawn.
            (bind _array_element (Utilities|Array|ForEachLoopwithBreak _output_get_2)
              (:LoopBody
                (if (or (.bSuccess (Utilities|Casting|CastToFortPlayerPawn _array_element)) (.bSuccess (Utilities|Casting|CastToFortAIPawn _array_element)))
                  (LocalVariables|SetTargetType "NewEnumerator2"))))))
        (:then_1
          (Reticle|Targeting|SetTargetType (Variables|Transient|GetTargetingComponent) _targettype)
          (Development|PrintString (Utilities|String|Append "MeleeTargeting: " (Utilities|String|EnumtoString _targettype)) true false "(R=1.000000,G=0.482353,B=0.000000,A=1.000000)" 1.0 "Melee_UpdateTargetingComponent"))))
    (:else)))
